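# Translucent, unlit, two-sided Niagara material for the wisp aura cosmetic, drawing glowing strands on spline-thickened geometry.
# Panning plasma noise, a U-edge falloff and a gradient mask swept by particle relative time form the wisp, tinted by particle color.
# A subtract-divide-saturate remap of that shape gives a tunable alpha.
# Vertex offset adds a time-driven sine wave to SplineThicken, whose width follows the SplineWidth dynamic parameter, both shaped by a sine along V.
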Material /SparksCosmetics/Items/Auras/Assets/NS_Aura_Wisp/M_Aura_Wisp_Spline
graph {
 properties {
  bAutomaticallySetUsageInEditor=false
  bUsedWithNiagaraMeshParticles=true
  bUsedWithNiagaraRibbons=true
  bUsedWithNiagaraSprites=true
  blendMode=BLEND_Translucent
  pixelDepthOffsetMode=PDOM_Legacy
  shadingModel=MSM_Unlit
  twoSided=true
  usedShadingModels=Unlit
 }
 nodes [
  n1 = MaterialFunctionCall() {materialFunction=&LinearGradient}
  n2 = Multiply(A=Time()) {constB=0.5}
  n3 = Sine(None=Multiply(A=Add(A=@n1.VGradient, B=@n2), B=ScalarParameter() {defaultValue=1.5}))
  n4 = VectorParameter() {defaultValue=<rgba 3 3 0 1>}
  n5 = Sine(None=Multiply(A=MaterialFunctionCall() {materialFunction=&LinearGradient}.VGradient))
  n6 = LinearInterpolate(B=DynamicParameter() {defaultValue=<rgba 8 1 1 1>, paramNames [SplineWidth,Param2,Param3,Param4]}.SplineWidth, Alpha=@n5) {constA=1}
  Add(
   A=Multiply(A=Multiply(A=@n3, B=@n4.RGB), B=@n5)
   B=MaterialFunctionCall(WidthBase=@n6, WidthTip=@n6) {materialFunction=&SplineThicken}.WorldPositionOffset
  )
  Constant() {r=1}
  n7 = TextureCoordinate()
  n8 = VectorParameter() {parameterName=Param_1, defaultValue=<rgba 0.4 0.5 0 0.5>}
  n9 = AppendVector(A=@n8.R, B=@n8.R)
  n10 = TextureSampleParameter2D(UVs=Panner(Coordinate=Multiply(A=@n7, B=@n9), Speed=AppendVector(A=@n8.B, B=@n8.B)) {speedY=0.5}) {
   parameterName=Texture
   samplerType=SAMPLERTYPE_LinearGrayscale
   texture=/Game/Effects/Fort_Effects/Emotes/UltraLight/Texture/T_Plasma_A_B_Greyscale
  }
  n11 = Frac(None=ParticleRelativeTime())
  n12 = Add(A=TextureCoordinate() {vTiling=1.5}, B=Subtract(A=Multiply(A=@n11) {constB=2}) {constB=0.8})
  n13 = MaterialFunctionCall() {materialFunction=&LinearGradient}
  n14 = ConstantBiasScale(None=Sine(None=Multiply(A=@n13.UGradient) {constB=0.5}))
  n15 = Multiply(
   A=@n10.R
   B=Multiply(
    A=TextureSample(UVs=@n12) {samplerSource=SSM_Clamp_WorldGroupSettings, texture=/Game/Effects/Fort_Effects/Textures/Masks/T_Vert_Gradient_Softer}.R
    B=Power(Base=@n14) {constExponent=7}
   )
  )
  MaterialFunctionCall(Emissive=Multiply(A=ParticleColor().RGB, B=Power(Base=@n15))) {materialFunction=&MF_FX_EmissiveAdjustments}
  n16 = ScalarParameter() {defaultValue=0.05, parameterName=Subtract}
  Saturate(None=Divide(A=Subtract(A=@n15, B=@n16) {constB=0.01}, B=ScalarParameter() {defaultValue=0.05, parameterName=Divide}) {constB=0.5})
  ScalarParameter() {defaultValue=8, parameterName=SplineThickness}
  Time()
 ]
}
